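; Projectile trajectory preview Blueprint for held object sockets, which sets up the spline mesh visuals for the arc.
; On BeginPlay it hides the Target component and pre-allocates a pool of spline mesh components up to MaxMeshCount.
; The first mesh creates a dynamic material with start and fade distances, and all later meshes share that one instance.
; Tick and overlap events are empty, and no code here shapes or updates the trajectory.
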(event EventBeginPlay
  (bind _maxmeshcount (Variables|Default|GetMaxMeshCount))
  (Rendering|SetVisibility (Variables|Default|GetTarget) false false)
  (Utilities|Array|Resize (Variables|Trajectory|GetSplineMeshComponents) _maxmeshcount)
  (for _index (range _maxmeshcount)
    (bind _returnvalue (AddComponent|NODE_AddSplineMeshComponent-0 :bManualAttachment true))
    (Utilities|Array|SetArrayElem (Variables|Trajectory|GetSplineMeshComponents) _index _returnvalue)
    (if (== _index 0)
      (bind _returnvalue_2 (Rendering|Material|CreateDynamicMaterialInstance _returnvalue 0 0 "None" :!class /Script/Engine.PrimitiveComponent))
      (bind _output_get (Variables|VFX|SetSplineMID _returnvalue_2))
      (Rendering|Material|SetScalarParameterValue _output_get "StartDistance" (Variables|Default|GetSplineFadeStartDistance) :!class /Script/Engine.MaterialInstanceDynamic)
      (Rendering|Material|SetScalarParameterValue _output_get "FadeDistance" (Variables|Default|GetSplineFadeDistance) :!class /Script/Engine.MaterialInstanceDynamic)
      (else
        (Rendering|Material|SetMaterial _returnvalue 0 (Variables|VFX|GetSplineMID))))))

(event Collision|EventActorBeginOverlap
  (params (OtherActor :type Actor%20Object%20Reference)))

(event EventTick
  (params (DeltaSeconds :type Float%20%28single-precision%29)))

(fn ConstructionScript
  (|CallParentFunction))
