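; B_HeldObjectSocket is an actor that captures a held object that hits it, provided the object matches its required gameplay tags.
; On hit, an empty socket validates the object, snaps it to the socket location, stops its projectile movement and registers itself as the holding socket.
; OnRep_SocketIsFilled calls empty filled or emptied hook events and updates tags on the socket manager, if valid.
; No logic here releases the object.
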(event EventBeginPlay
  (Collision|BindEventtoOnActorHit (delegate Custom|OnHitHeldObject)))

(event Custom|OnHitHeldObject
  (params (SelfActor :type Actor%20Object%20Reference) (OtherActor :type Actor%20Object%20Reference) (NormalImpulse :type Vector) (Hit :type Hit%20Result%20Structure%20%28by%20ref%29))
  (if (not (Variables|Default|GetHasSocketedHeldObject))
    (bind _asb_held_object_parent (Utilities|Casting|CastToB_HeldObject_Parent OtherActor))
    (bind (_heldobject_output _isvalid) (CallFunction|ValidateHeldObject _asb_held_object_parent))
    (if _isvalid
      (bind _output_get (Variables|Default|SetHeldObject _heldobject_output))
      (Transformation|AttachActorToComponent _output_get (Variables|Default|GetSocketLocation) "None" "SnapToTarget" "SnapToTarget" "SnapToTarget" true)
      (Components|Movement|StopMovementImmediately (Variables|/Game/Athena/Items/HeldObjects/B_HeldObject_Parent.B_HeldObject_Parent_C|GetFortProjectileMovement _output_get))
      (Variables|/Game/Athena/Items/HeldObjects/B_HeldObject_Parent.B_HeldObject_Parent_C|SetHoldingSocket self (Variables|Default|GetHeldObject))
      (Variables|Default|SetHasSocketedHeldObject true))))

(event Custom|EventOnSocketFilled)

(event Custom|EventOnSocketEmptied)

(fn OnRep_SocketIsFilled
  (bind _heldobjectsocketmanager (Variables|Default|GetHeldObjectSocketManager))
  (if (Variables|Default|GetHasSocketedHeldObject)
    (CallFunction|EventonSocketFilled)
    (else
      (CallFunction|EventonSocketEmptied)))
  (Utilities|IsValid _heldobjectsocketmanager
    (:Is%20Valid
      (Class|BHeldObjectSocketManager|UpdateHeldObjectTags _heldobjectsocketmanager self))))

(fn ConstructionScript)

(fn ValidateHeldObject
  (params (HeldObject%20Input :type B%20Held%20Object%20Parent%20Object%20Reference))
  (return HeldObject%20Input (GameplayTags|TagContainer|CompareTagAssetInterfacetoOtherTagContainers (Variables|Default|GetRequiredHeldObjectTags) HeldObject%20Input)))
