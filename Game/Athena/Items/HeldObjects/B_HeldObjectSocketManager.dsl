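; Held-object socket manager that checks whether the objects placed in its sockets form the correct gameplay-tag combination.
; On BeginPlay it registers with each socket; UpdateHeldObjectTags records a changed socket's object tags at that socket's index, then re-verifies.
; Verification compares containers per index when order matters, otherwise merges each side into one container before comparing.
; The result sets CombinationCorrect, whose RepNotify fires the empty OnCombinationCorrect hook when true.
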(event EventBeginPlay
  (for _array_element (Variables|Default|GetHeldObjectSockets)
    (Variables|/Game/Athena/Items/HeldObjects/B_HeldObjectSocket.B_HeldObjectSocket_C|SetHeldObjectSocketManager self _array_element)))

(event Custom|OnCombinationCorrect)

(fn CheckTagMatches
  (bind _isvalidmatch (CallFunction|VerifyHeldObjectTags (Variables|Default|GetDefaultCorrectTags) (Variables|Default|GetCurrentTags) (Variables|Default|GetSocketOrderMatters)))
  (Variables|Default|SetCombinationCorrect _isvalidmatch))

(fn OnRep_CombinationCorrect
  (if (Variables|Default|GetCombinationCorrect)
    (CallFunction|OnCombinationCorrect)))

(fn UpdateHeldObjectTags
  (params (ChangedSocket :type B%20Held%20Object%20Socket%20Object%20Reference))
  (bind _heldobjectsockets (Variables|Default|GetHeldObjectSockets))
  (bind _returnvalue (Utilities|Array|FindItem _heldobjectsockets ChangedSocket))
  (bind _currenttags (Variables|Default|GetCurrentTags))
  (if (Variables|/Game/Athena/Items/HeldObjects/B_HeldObjectSocket.B_HeldObjectSocket_C|GetHasSocketedHeldObject ChangedSocket)
    (Utilities|Array|SetArrayElem _currenttags _returnvalue (GameplayTags|GetOwnedGameplayTags (Variables|/Game/Athena/Items/HeldObjects/B_HeldObjectSocket.B_HeldObjectSocket_C|GetHeldObject ChangedSocket) :!class /Script/GameplayTags.GameplayTagAssetInterface) true)
    (else
      (Utilities|Array|SetArrayElem :TargetArray _currenttags :Index _returnvalue :bSizeToFit true)))
  (CallFunction|CheckTagMatches))

(fn ConstructionScript)

(fn VerifyHeldObjectTags
  (params (CorrectTags :type Array%20of%20Gameplay%20Tag%20Container%20Structures) (CurrentTags :type Array%20of%20Gameplay%20Tag%20Container%20Structures) (OrderMatters :type Boolean))
  (if OrderMatters
    (LocalVariables|SetTagsAreInOrder true)
    (bind (_array_element _array_index) (Utilities|Array|ForEachLoopwithBreak CorrectTags)
      (:LoopBody
        (if (> (Utilities|Array|Length CurrentTags) _array_index)
          (if (GameplayTags|TagContainer|CompareTagContainertoOtherTagContainers _array_element (Utilities|Array|Get(aref) CurrentTags _array_index))
            _
            (else
              _))
          (else
            _))
        (LocalVariables|SetTagsAreInOrder false))
      (:Completed
        (return (LocalVariables|GetTagsAreInOrder))))
    (else
      (LocalVariables|SetConcatenatedCorrectTags)
      (LocalVariables|SetConcatenatedCurrentTags)
      (for _array_element_2 CorrectTags
        (GameplayTags|AppendGameplayTagContainers (LocalVariables|GetConcatenatedCorrectTags) _array_element_2))
      (for _array_element_3 CurrentTags
        (GameplayTags|AppendGameplayTagContainers (LocalVariables|GetConcatenatedCurrentTags) _array_element_3))
      (return (GameplayTags|TagContainer|CompareTagContainertoOtherTagContainers (LocalVariables|GetConcatenatedCorrectTags) (LocalVariables|GetConcatenatedCurrentTags))))))
