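; Parent actor for held objects that player pawns can pick up, carry, release and place in a B_HeldObjectSocket.
; Interaction records the holder and sets IsHeld, unless the object is already held or the pawn has a blocking tag.
; OnRep_IsHeld either attaches and hides the object and applies an ability-weapon GameplayEffect, or detaches it and enables projectile movement ignoring the holder.
; OnRep_IsSocketed rejects invalid sockets and clears socket references on removal.
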(event EventBeginPlay
  (Components|Movement|SetUpdatedComponent (Variables|Default|GetFortProjectileMovement) 0))

(event Interaction|EventBlueprintOnInteract
  (params (InteractingPawn :type Fort%20Pawn%20Object%20Reference) (InteractionBeingAttempted :type EInteractionBeingAttempted%20Enum :default "FirstInteraction"))
  (bind _asfort_player_pawn (Utilities|Casting|CastToFortPlayerPawn InteractingPawn))
  (Variables|Default|SetHoldingPawn _asfort_player_pawn)
  (Variables|Default|SetIsHeld true)
  (Variables|Default|SetIsSocketed false))

(fn BlueprintCanInteract
  (params (InteractingPawn :type Fort%20Pawn%20Object%20Reference) (InteractionBeingAttempted :type EInteractionBeingAttempted%20Enum :default "FirstInteraction") (InteractionType :type TInteractionType%20Enum :default "IT_NoInteraction"))
  (return (and (not (Variables|Default|GetIsHeld)) (not (GameplayTags|HasAnyMatchingGameplayTags InteractingPawn (Variables|Default|GetTC_BlockPickup))))))

(fn BlueprintGetInteractionString
  (params (InteractingPawn :type Fort%20Pawn%20Object%20Reference) (InteractionBeingAttempted :type EInteractionBeingAttempted%20Enum :default "FirstInteraction"))
  (return "NSLOCTEXT(\"[E0C4A5DB4BB5027EADA154B9BDF4E8E6]\", \"E05FABA14ECC5C99D3EB6F9655989C96\", \"Pick up\")"))

(fn OnRep_IsHeld
  (bind _holdingpawn (Variables|Default|GetHoldingPawn))
  (if (Variables|Default|GetIsHeld)
    (Transformation|AttachActorToActor _holdingpawn "None" "KeepWorld" "KeepWorld" "KeepWorld" false)
    (Transformation|SetActorRelativeLocation (Variables|Default|GetVector_HoldingOffset))
    (Rendering|SetVisibility (Variables|Default|GetStaticMesh) false false)
    (Collision|SetCollisionEnabled (Variables|Default|GetCapsule) "NoCollision")
    (Environment|ApplyGlobalEnvironmentGameplayEffecttoActor _holdingpawn "/Game/Athena/Items/HeldObjects/GE_HeldObject_AbilityWeapon.GE_HeldObject_AbilityWeapon_C")
    (Variables|Default|SetIsSocketed false)
    (else
      (Transformation|DetachFromActor "KeepWorld" "KeepWorld" "KeepWorld")
      (Collision|AddIgnoredActor _holdingpawn self 1.0)
      (Rendering|SetVisibility (Variables|Default|GetStaticMesh) true true)
      (Collision|SetCollisionEnabled (Variables|Default|GetCapsule) "QueryOnly")
      (Components|Movement|SetUpdatedComponent (Variables|Default|GetFortProjectileMovement) (Variables|Default|GetCapsule))
      (Collision|IgnoreActorwhenMoving (Variables|Default|GetCapsule) _holdingpawn true)
      (Variables|Default|SetHoldingPawn 0))))

(fn OnRep_IsSocketed
  (bind _holdingsocket (Variables|Default|GetHoldingSocket))
  (if (Variables|Default|GetIsSocketed)
    (Utilities|IsValid _holdingsocket
      (:Is%20Not%20Valid
        (Variables|Default|SetIsSocketed false)))
    (else
      (Utilities|IsValid _holdingsocket
        (:Is%20Valid
          (Variables|/Game/Athena/Items/HeldObjects/B_HeldObjectSocket.B_HeldObjectSocket_C|SetHasSocketedHeldObject false _holdingsocket)
          (Variables|/Game/Athena/Items/HeldObjects/B_HeldObjectSocket.B_HeldObjectSocket_C|SetHeldObject 0 _holdingsocket)
          (Variables|Default|SetHoldingSocket 0))))))

(fn ConstructionScript)
