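; Gameplay ability that equips an ability weapon on the player pawn while it carries a held object.
; On activation it caches the avatar pawn, equips the weapon and waits for the Event.HeldObject.Released gameplay event to end itself.
; On end it filters the pawn's attached actors for B_HeldObject_Parent and clears IsHeld on the first match.
; A failed passive activation also ends the ability.
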(event Ability|EventActivateAbility
  (Variables|Default|SetPlayerPawn (Utilities|Casting|CastToFortPlayerPawn (Ability|GetAvatarActorfromActorInfo)))
  (AbilityWeapon|EquipAbilityWeapon (Variables|AbilityWeapon|GetAbilityWeapon))
  (Ability|Tasks|WaitGameplayEvent "(TagName=\"Event.HeldObject.Released\")"
    (:EventReceived
      (Ability|EndAbility))))

(event Ability|EventOnEndAbility
  (params (bWasCancelled :type Boolean))
  (bind _filteredarray (Utilities|Array|FilterArray (Actor|GetAttachedActors (Variables|Default|GetPlayerPawn) true false) "/Game/Athena/Items/HeldObjects/B_HeldObject_Parent.B_HeldObject_Parent_C"))
  (bind _asb_held_object_parent (Utilities|Casting|CastToB_HeldObject_Parent (Utilities|Array|Get(acopy) _filteredarray)))
  (Variables|/Game/Athena/Items/HeldObjects/B_HeldObject_Parent.B_HeldObject_Parent_C|SetIsHeld false _asb_held_object_parent))

(event Ability|EventFailedToActivatePassiveAbility
  (params (ActorInfo :type Gameplay%20Ability%20Actor%20Info%20Structure))
  (|CallParentFunction)
  (Ability|EndAbility))
