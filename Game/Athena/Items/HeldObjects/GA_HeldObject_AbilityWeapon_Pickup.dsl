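; Gameplay ability for the pickup of a held-object ability weapon, which plays the character's pickup animation.
; On activation, a PlayMontageAndWait task plays the HeldObject_Pickup_CMM_M montage, followed by an EndAbility call.
; The graph holds no other gameplay logic, such as attaching the object or applying effects.
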(event Ability|EventActivateAbility
  (Ability|Tasks|PlayMontageAndWait "None" "/Game/Animation/Game/MainPlayer/Combat/Gadgets/Medium/Male/HeldObjects/HeldObject_Pickup_CMM_M.HeldObject_Pickup_CMM_M"
    (:OnCompleted
      _)
    (:OnBlendOut
      _))
  (Ability|EndAbility))
