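; GA_HeldObject_Drop is a gameplay ability that drops the object a player is holding.
; On activation it marks itself non-cancelable, sends the Event.HeldObject.Released gameplay event, then ends immediately.
; The release itself is left to whatever listens for that event, and the end-ability handler is empty.
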(event Ability|EventActivateAbility
  (Ability|SetCanbeCanceled false)
  (Ability|SendGameplayEvent "(TagName=\"Event.HeldObject.Released\")")
  (Ability|EndAbility))

(event Ability|EventOnEndAbility
  (params (bWasCancelled :type Boolean)))
