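; Gameplay ability that throws the player's attached held object, with an aim trajectory preview.
; Activation finds the object and starts windup; input release plays the throw montage, whose trigger leads to a server-side commit.
; The server launches the existing actor via its projectile movement, then sends a released event after a delay.
; The preview simulates the arc from the object's speed, gravity and capsule, switching indicators over a valid empty socket.
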(event Ability|EventActivateAbility
  (ParentDefaults|PrethrowSetup) ; Ability is instanced per actor, so we need to reset variables\nevery time the ability is activated.
  (TrajectoryInterface|SpawnTrajectoryIndicator false :!class /Script/FortniteGame.FortTrajectoryAbilityInterface))

(event TrajectoryInterface|EventInitTrajectoryVariables
  (bind (_outradius _outhalfheight) (Components|Capsule|GetScaledCapsuleSize (Variables|/Game/Athena/Items/HeldObjects/B_HeldObject_Parent.B_HeldObject_Parent_C|GetCapsule (Variables|Default|GetHeldObject))))
  (bind _heldobject (Variables|Default|GetHeldObject))
  ;[1 1 1] Spawn a dummy projectile to check the projectile's max speed, gravity scale, and sphere extent. Used by the trajectory display.
  (Variables|TrajectoryDisplay|SetDummyMaxSpeed (Variables|/Game/Athena/Items/HeldObjects/B_HeldObject_Parent.B_HeldObject_Parent_C|GetThrowSpeed _heldobject))
  (Variables|TrajectoryDisplay|SetDummyGravity (Components|Movement|GetGravityZ (Variables|/Game/Athena/Items/HeldObjects/B_HeldObject_Parent.B_HeldObject_Parent_C|GetFortProjectileMovement _heldobject)))
  (Variables|TrajectoryDisplay|SetDummyExtent (Math|Float|Max(Float) _outradius _outhalfheight)))

(event Ability|EventOnAbilityInputReleased
  ;[1 1 1] Throw the projectile if input is released while still in throw windup.
  (Variables|AbilityState|SetAbilityKeyPressed false)
  (if (Variables|AbilityState|GetInThrowWindup)
    (CallFunction|ThrowProjectile)))

(event Custom|ThrowProjectile
  ;[1 1 1] BR: ShouldBlockAbilities = true
  ; StW: CanBeCanceled = false,  ShouldBlockAbilities = true
  (Ability|SetShouldBlockOtherAbilities true)
  (ParentDefaults|PostThrowCleanup)
  ;[1 1 1] Projectile is spawned when the montage reaches ability notify
  (Ability|Tasks|PlayMontageandWaitforTrigger "None" (Variables|Default|GetThrowMontage)
    (:Triggered
      (CallFunction|AttemptSpawnThrownProjectile))
    (:Cancelled
      (Ability|EndAbility))
    (:Completed
      (Ability|SetShouldBlockOtherAbilities false))))

(event Custom|ServerSpawnThrownProjectile
  (if (Networking|IsServer)
    (CallFunction|ThrowProjectile)))

(event Custom|AttemptSpawnThrownProjectile
  (bind (_location _rotation) (ParentDefaults|GetSpawnLocationandRotation))
  (bind _heldobject (Variables|Default|GetHeldObject))
  (Ability|ExecuteGameplayCueOnOwner (Variables|Default|GetGC_OnThrow)) ; GC to play on the player when the projectile is thrown
  (if (Networking|IsServer)
    ;[1 1 1] Commit the ability here to prevent exploits relating to spawning multiple projectiles
    (bind _returnvalue (Ability|CommitAbility))
    (if _returnvalue
      (Variables|ParentDefaults|SetExplosionEffectContainerSpec (Ability|MakeGameplayEffectContainerSpec (Variables|ParentDefaults|GetEC_DefaultExplosion)))
      ;[1 1 1] Override to modify projectile spawn inputs
      (CallFunction|OnHeldObjectSetup)
      (Variables|/Game/Athena/Items/HeldObjects/B_HeldObject_Parent.B_HeldObject_Parent_C|SetIsHeld false _heldobject)
      (Transformation|SetActorLocation _heldobject _location false false)
      (Variables|/Script/Engine.MovementComponent|SetVelocity (* (Math|Rotator|GetRotationXVector _rotation) (Variables|/Game/Athena/Items/HeldObjects/B_HeldObject_Parent.B_HeldObject_Parent_C|GetThrowSpeed _heldobject)) (Variables|/Game/Athena/Items/HeldObjects/B_HeldObject_Parent.B_HeldObject_Parent_C|GetFortProjectileMovement _heldobject))
      ;[1 1 1] Override to modify replicated variables on the spawned projectile
      (CallFunction|OnHeldObjectThrown 0)
      (Ability|Tasks|WaitDelay 0.4
        (:OnFinish
          (Ability|SendGameplayEvent "(TagName=\"Event.HeldObject.Released\")")
          (Ability|AbilityCompleted) ; We only execute this part of the event on the server,\nso is this meaningfully distinguishable from EndAbility?
        )) ; Would prefer to end the ability without a delay if possible.
      (else
        (Ability|EndAbility)))
    (else
      (CallFunction|ServerSpawnThrownProjectile))))

(event TrajectoryInterface|EventUpdateTrajectory
  (TrajectoryDisplay|UpdateTrajectorySpline))

(event Ability|EventOnEndAbility
  (params (bWasCancelled :type Boolean))
  ;[1 1 1] Failsafe to clean up the trajectory when the ability ends
  (|CallParentFunction bWasCancelled)
  (ParentDefaults|PostThrowCleanup))

(event TrajectoryInterface|EventCleanupTrajectoryIndicatorOnUnequip
  (TrajectoryDisplay|CleanupTrajectory))

(fn CleanupTrajectory
  (Class|BPProjectileTrajectory|SetShouldUpdatefromOwner (Variables|TrajectoryDisplay|GetTrajectoryIndicator) false 0)
  (Class|BPProjectileTrajectory|SetShouldUpdatefromOwner (Variables|Default|GetProjectileTrajectorySocketed) false 0)
  (Actor|DestroyActor (Variables|Default|GetProjectileTrajectorySocketed)))

(fn GetOwningPlayer
  (bind (_owneractor _avataractor) (Utilities|Struct|BreakGameplayAbilityActorInfo (Ability|GetActorInfo)))
  (Utilities|IsValid (Variables|ReferencedActors|GetPlayerPawn)
    (:Is%20Valid
      (return (Variables|ReferencedActors|GetPlayerPawn)))
    (:Is%20Not%20Valid
      (bind _asfort_player_pawn (Utilities|Casting|CastToFortPlayerPawn _avataractor)
        (:then
          (return _asfort_player_pawn))
        (:CastFailed
          (return 0))))))

(fn GetProjectileTrajectoryActor
  (return (Variables|TrajectoryDisplay|GetTrajectoryIndicatorClass)))

(fn GetProjectileTrajectoryPoints
  (bind (_location _rotation) (ParentDefaults|GetSpawnLocationandRotation))
  (bind (_outhitresult _outsplinepoints _outsplinetangents) (Projectile|CalculateProjectileTrajectorySplinewithHit :Instigator (ParentDefaults|GetOwningPlayer) :InitialLocation _location :InitialVelocity (* (Math|Rotator|GetRotationXVector _rotation) (Variables|TrajectoryDisplay|GetDummyMaxSpeed)) :MaxSpeed (Variables|TrajectoryDisplay|GetDummyMaxSpeed) :Gravity (Variables|TrajectoryDisplay|GetDummyGravity) :Friction 0.0 :Bounciness 0.0 :TimeStep (Variables|TrajectoryDisplay|GetDummyTimeStep) :TraceExtent (Variables|TrajectoryDisplay|GetDummyExtent) :TraceChannel "ECC_GameTraceChannel8" :MaxBounces 1 :MaxSteps 70 :MaxDistanceBetweenSplinePoints 1.5e+03 :InitialDistance 0.0 :LinearDamping 0.0 :bReturnPhysicalMaterial false))
  (return _outsplinepoints _outsplinetangents _outhitresult))

(fn GetSpawnLocationAndRotation
  (return (+ (Transformation|GetActorLocation (ParentDefaults|GetOwningPlayer)) (Math|Vector|RotateVectorAroundAxis (select (Variables|/Script/Engine.Character|GetbIsCrouched (ParentDefaults|GetOwningPlayer)) (Variables|ParentDefaults|GetGrenadeTargetingOriginOffset_Crouched) (Variables|ParentDefaults|GetGrenadeTargetingOriginOffset)) (.yaw (Transformation|GetActorRotation (ParentDefaults|GetOwningPlayer))) "0,0.000000,1.000000")) (Math|Rotator|MakeRotator (.roll (Targeting|GetDefaultTargetingRotation (ParentDefaults|GetOwningPlayer))) (Math|Float|Min(Float) (Math|Rotator|NormalizeAxis (+ (.pitch (Targeting|GetDefaultTargetingRotation (ParentDefaults|GetOwningPlayer))) (Variables|ParentDefaults|GetAdditionalThrowAngle))) (Variables|ParentDefaults|GetMaxTossPitch)) (.yaw (Targeting|GetDefaultTargetingRotation (ParentDefaults|GetOwningPlayer))))))

(fn HitValidSocket
  (params (HitResult :type Hit%20Result%20Structure))
  (bind (_bblockinghit _binitialoverlap _time _distance _location _impactpoint _normal _impactnormal _physmat _hitactor _hitcomponent) (Collision|BreakHitResult HitResult))
  (Utilities|IsValid _hitactor
    (:Is%20Valid
      (bind _asb_held_object_socket (Utilities|Casting|CastToB_HeldObjectSocket _hitactor)
        (:then
          (Utilities|Casting|CastToSphereComponent _hitcomponent
            (:then
              (bind (_heldobject_output _isvalid) (Class|BHeldObjectSocket|ValidateHeldObject _asb_held_object_socket (Variables|Default|GetHeldObject)))
              (return (and _isvalid (not (Variables|/Game/Athena/Items/HeldObjects/B_HeldObjectSocket.B_HeldObjectSocket_C|GetHasSocketedHeldObject _asb_held_object_socket)))))
            (:CastFailed
              _)))
        (:CastFailed
          _))
      (return false))
    (:Is%20Not%20Valid
      _)))

(fn OnHeldObjectSetup)

(fn OnHeldObjectThrown
  (params (SpawnedProjectile :type Fort%20Projectile%20Base%20Object%20Reference)))

(fn PostThrowCleanup
  ;[1 1 1] Clear the looping trajectory display timer, destroy the trajectory indicator, and remove windup GE.
  (Variables|AbilityState|SetInThrowWindup false)
  (Ability|RemoveGameplayEffectFromOwnerWithHandle (Variables|ParentDefaults|GetHandle_GE_InWindup))
  (if (not (Variables|Default|GetSpawnedOnEquip))
    (TrajectoryDisplay|CleanupTrajectory)))

(fn PrethrowSetup
  (bind (_owneractor _avataractor) (Utilities|Struct|BreakGameplayAbilityActorInfo (Ability|GetActorInfo)))
  (bind _asfort_player_pawn (Utilities|Casting|CastToFortPlayerPawn _avataractor)
    (:then
      (Variables|ReferencedActors|SetPlayerPawn _asfort_player_pawn)
      ;[1 1 1] We allow a player to cancel out of the windup for most thrown projectiles
      (Ability|SetShouldBlockOtherAbilities false)
      (Ability|SetCanbeCanceled true)
      ;[1 1 1] The input is being held down and the actor is winding up to throw the projectile
      (Variables|AbilityState|SetAbilityKeyPressed true)
      (Variables|AbilityState|SetInThrowWindup true)
      ;[1 1 1] Add a GE that applies looping GC while in the windup
      (bind _returnvalue (Ability|ApplyGameplayEffectToOwner (Variables|Default|GetGE_InWindup)))
      (Variables|ParentDefaults|SetHandle_GE_InWindup _returnvalue)
      (bind _filteredarray (Utilities|Array|FilterArray (Actor|GetAttachedActors (ParentDefaults|GetOwningPlayer) true false) "/Game/Athena/Items/HeldObjects/B_HeldObject_Parent.B_HeldObject_Parent_C"))
      (bind _asb_held_object_parent (Utilities|Casting|CastToB_HeldObject_Parent (Utilities|Array|Get(acopy) _filteredarray))
        (:then
          (Variables|Default|SetHeldObject _asb_held_object_parent))
        (:CastFailed
          (Ability|EndAbility))))
    (:CastFailed
      (Ability|EndAbility))))

(fn SetupDummyProjectile
  (bind _trajectoryindicatorclass (Variables|TrajectoryDisplay|GetTrajectoryIndicatorClass))
  (if (Utilities|IsValidClass _trajectoryindicatorclass)
    (bind _returnvalue (Game|SpawnActorfromClass _trajectoryindicatorclass (Transformation|GetActorTransform (ParentDefaults|GetOwningPlayer)) "AlwaysSpawn" "SelectDefaultAtRuntime" 0 (ParentDefaults|GetOwningPlayer)))
    (bind _asbp_projectile_trajectory (Utilities|Casting|CastToBP_ProjectileTrajectory _returnvalue))
    (Variables|TrajectoryDisplay|SetTrajectoryIndicator _asbp_projectile_trajectory)
    (TrajectoryInterface|InitTrajectoryVariables :!class /Script/FortniteGame.FortTrajectoryAbilityInterface)))

(fn SetupDummyProjectileSocketed
  (Utilities|IsValid (Variables|Default|GetProjectileTrajectorySocketed)
    (:Is%20Not%20Valid
      (bind _returnvalue (Game|SpawnActorfromClass (Variables|Default|GetTrajectoryIndicatorSocketedClass) (Transformation|GetActorTransform (ParentDefaults|GetOwningPlayer)) "AlwaysSpawn" "SelectDefaultAtRuntime" 0 0))
      (bind _asbp_projectile_trajectory (Utilities|Casting|CastToBP_ProjectileTrajectory _returnvalue))
      (Variables|Default|SetProjectileTrajectorySocketed _asbp_projectile_trajectory))))

(fn ShouldOnlyShowTrajectoryOnUse
  (return false))

(fn SpawnThrownProjectile
  (params (ProjectileClass :type Fort%20Projectile%20Base%20Class%20Reference) (ProjectileSpawnLocation :type Vector%20%28by%20ref%29) (ProjectileSpawnDirection :type Rotator%20%28by%20ref%29) (ExplosionEffectContainerSpec :type Fort%20Gameplay%20Effect%20Container%20Spec%20Structure) (ProjectileInitialSpeed :type Float%20%28double-precision%29) (ProjectileGravityScale :type Float%20%28double-precision%29))
  (bind _returnvalue (Projectile|SpawnProjectile :ProjectileClass ProjectileClass :RequestedBy (ParentDefaults|GetOwningPlayer) :SpawnLocation ProjectileSpawnLocation :SpawnRotation ProjectileSpawnDirection :SpawnDirection ProjectileSpawnDirection :EffectContainerSpecToApplyOnExplode ExplosionEffectContainerSpec :bUseDefaultPhysics false :InitialSpeed ProjectileInitialSpeed :GravityScale ProjectileGravityScale :ChargePercent 0.0))
  (bind _output_get (Variables|ReferencedActors|SetSpawnedProjectile _returnvalue))
  (return _output_get))

(fn SpawnTrajectoryIndicator
  (params (bSpawnOnEquip :type Boolean))
  (bind _player_pawn (ParentDefaults|GetOwningPlayer))
  (bind _returnvalue (Pawn|IsLocallyControlled _player_pawn))
  (CallFunction|SetupDummyProjectileSocketed)
  (Utilities|IsValid (Variables|TrajectoryDisplay|GetTrajectoryIndicator)
    (:Is%20Not%20Valid
      (Variables|Default|SetSpawnedOnEquip bSpawnOnEquip)
      (Utilities|FlowControl|Sequence
        (:then_0
          (if _returnvalue
            (TrajectoryDisplay|SetupDummyProjectile)))
        (:then_1
          ;[1 1 1] Only show trajectory display if input is held down
          (if (and _returnvalue (or (Variables|AbilityState|GetAbilityKeyPressed) (Variables|Default|GetSpawnedOnEquip)))
            (Class|BPProjectileTrajectory|SetShouldUpdatefromOwner (Variables|TrajectoryDisplay|GetTrajectoryIndicator) true self)
            (return (Variables|TrajectoryDisplay|GetTrajectoryIndicator))))))))

(fn UpdateTrajectorySpline
  (bind _trajectoryindicator (Variables|TrajectoryDisplay|GetTrajectoryIndicator))
  (bind _projectiletrajectorysocketed (Variables|Default|GetProjectileTrajectorySocketed))
  (if (and (Utilities|IsValid (Variables|TrajectoryDisplay|GetTrajectoryIndicator) :!class /Script/Engine.KismetSystemLibrary) (Utilities|IsValid (Variables|Default|GetProjectileTrajectorySocketed) :!class /Script/Engine.KismetSystemLibrary))
    (bind (_outsplinepoints _outsplinetangents _outhitresult) (TrajectoryInterface|GetProjectileTrajectoryPoints :!class /Script/FortniteGame.FortTrajectoryAbilityInterface))
    (bind _cansocket (CallFunction|HitValidSocket _outhitresult))
    (Trajectory|SetTrajectorySpline (select _cansocket _projectiletrajectorysocketed _trajectoryindicator) _outsplinepoints _outsplinetangents)
    (Rendering|SetActorHiddenInGame (select _cansocket _projectiletrajectorysocketed _trajectoryindicator) false)
    (Rendering|SetActorHiddenInGame (select _cansocket (Variables|TrajectoryDisplay|GetTrajectoryIndicator) (Variables|Default|GetProjectileTrajectorySocketed)) true)))
