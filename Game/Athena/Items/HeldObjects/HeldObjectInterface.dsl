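; Blueprint interface for held objects in the Athena game mode's items.
; It declares a single function, GetLastInstigator, which lets callers query a held object for its most recent instigator.
; The function body here holds only an empty default return, with no logic of its own.
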(fn GetLastInstigator
  (return 0))
