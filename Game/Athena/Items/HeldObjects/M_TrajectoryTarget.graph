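# Unlit translucent material for the trajectory target of held objects, drawn as a pulsing, tinted marker with scrolling scanlines.
# Panned, tiled UVs are thresholded into faint hard-edged horizontal bands and added to an inverted Fresnel term.
# A sine of time pulses the Color parameter, and an inverted DepthFade boosts it near intersections with scene geometry.
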Material /Game/Athena/Items/HeldObjects/M_TrajectoryTarget
graph {
 properties {blendMode=BLEND_Translucent, pixelDepthOffsetMode=PDOM_Legacy, shadingModel=MSM_Unlit, usedShadingModels=Unlit}
 nodes [
  n1 = Constant() {r=1}
  n2 = ScalarParameter() {defaultValue=32, parameterName=ScanlineTileHorizontal}
  n3 = Panner(Coordinate=Multiply(
    A=TextureCoordinate()
    B=MaterialFunctionCall(X=@n1, Y=@n2) {materialFunction=&MakeFloat2}.Result
   )) {speedY=-5}
  n4 = Add(A=Frac(None=ComponentMask(None=@n3) {g=true})) {constB=0.5}
  Add(
   A=MaterialFunctionCall(Exponent=Constant() {r=2}) {materialFunction=&MF_Fresnel_FX}.Fresnel Inverted
   B=Multiply(A=Floor(None=@n4)) {constB=0.03}
  )
  n5 = Time()
  n6 = Multiply(A=Sine(None=Multiply(A=@n5) {constB=1.8})) {constB=0.25}
  n7 = DepthFade(FadeDistance=Constant() {r=16})
  Multiply(A=Multiply(A=Add(A=@n6), B=VectorParameter() {parameterName=Color, defaultValue=<rgba 1.5 1.5 0 0>}.RGB), B=Add(A=OneMinus(None=@n7)))
 ]
}
